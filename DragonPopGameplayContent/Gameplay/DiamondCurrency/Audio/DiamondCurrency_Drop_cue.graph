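# Sound cue for the diamond currency drop in DragonPop gameplay, playing the DillCoin currency landing sound.
# A single wave player feeds a Modulator with pitch randomization flattened, leaving volume modulation at its default.
# It borrows the shared Fort weapon drop attenuation and routes to the loot collect sound class at reduced volume.
# A per-cue concurrency override tightly caps simultaneous plays so clustered drops do not pile up.
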SoundCue /DragonPopGameplayContent/Gameplay/DiamondCurrency/Audio/DiamondCurrency_Drop_cue
root=Modulator_0
nodes [
 {id=Modulator_0, children [WavePlayer_0], properties {pitchMax=1, pitchMin=1}}
 {id=WavePlayer_0, properties {soundWaveAssetPtr=&/DillCoin_Currency_Land}}
]
properties {
 attenuationSettings=&Fort_Weapon_Drop_Att
 bOverrideConcurrency=true
 maxDistance=3000
 priority=0.9
 soundClassObject=&Fort_Loot_Collect
 volumeMultiplier=0.5
 attenuationOverrides {falloffDistance=1000, stereoSpread=512, attenuationShapeExtents {x=0}}
 concurrencyOverrides {maxCount=2}
 concurrencySet [null]
}
