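# Sound cue for the diamond currency pickup, reusing the DillCoin currency pickup wave.
# A single Wave Player feeds a Modulator that applies a fixed raised pitch and unchanged volume.
# It routes to the loot-collect sound class with weapon-drop attenuation assigned.
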SoundCue /DragonPopGameplayContent/Gameplay/DiamondCurrency/Audio/DiamondCurrency_Pickup_Cue
root=Modulator_0
nodes [
 {id=Modulator_0, children [WavePlayer_3], properties {pitchMax=1.3, pitchMin=1.3, volumeMax=1, volumeMin=1}}
 {id=WavePlayer_3, properties {soundWaveAssetPtr=&/DillCoin_Currency_PickUp}}
]
properties {
 attenuationSettings=&Fort_Weapon_Drop_Att
 maxDistance=3000
 priority=0.9
 soundClassObject=&Fort_Loot_Collect
 volumeMultiplier=0.8
 attenuationOverrides {falloffDistance=1000, stereoSpread=512, attenuationShapeExtents {x=0}}
 concurrencyOverrides {maxCount=1, resolutionRule=PreventNew}
 concurrencySet [null]
}
