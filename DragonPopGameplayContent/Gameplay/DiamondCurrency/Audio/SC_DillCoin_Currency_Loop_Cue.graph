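# Sound cue for the looping audio of the DillCoin currency item in the DragonPop diamond currency gameplay content.
# A looping wave player feeds an Enveloper whose volume curve fades the loop in over about a second.
# The cue uses DillCoin loop attenuation and the Fort_Consumables sound class, with a lowered volume and a concurrency override limiting how many copies play at once.
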SoundCue /DragonPopGameplayContent/Gameplay/DiamondCurrency/Audio/SC_DillCoin_Currency_Loop_Cue
root=Enveloper_0
nodes [
 {
  id=Enveloper_0
  children [WavePlayer_2]
  properties {pitchCurve {editorCurveData {keys [{time=0}]}}, volumeCurve {editorCurveData {keys [<at 0 0>,<at 0.5879 0.2712>,<at 0.9825 0.9092>]}}}
 }
 {
  id=WavePlayer_2
  properties {bLooping=true, soundWaveAssetPtr=&/DillCoin_Currency_Loop_A}
 }
]
properties {
 attenuationSettings=&/ATT_DillCoin_loop
 bOverrideConcurrency=true
 maxDistance=4000
 soundClassObject=&Fort_Consumables
 volumeMultiplier=0.4
 concurrencyOverrides {maxCount=6}
}
