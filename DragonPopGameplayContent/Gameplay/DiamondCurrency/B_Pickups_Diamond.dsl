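; Pickup actor for the DragonPop diamond currency item, handling its world look and collection feedback.
; On attach, cosmetic-capable instances cache the parent pickup's stack count, apply a CoinLawless mesh and adjust its transform.
; Pickup spawns a green rarity Niagara burst; backpack entry attaches the sound to the pawn, with debug prints remaining.
; Helpers for threshold-based wad size, leveled glow and server-side marking of the nearest black market icon go uncalled here.
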(event EventBeginPlay
  (|CallParentFunction))

(event EventOnTossed
  (|CallParentFunction))

(event EventOnAttached
  (bind _returnvalue (Actor|GetAttachParentActor))
  (bind _staticmeshcomponent (Variables|FortPickupsParent|GetStaticMeshComponent))
  (bind (_relativelocation_x _relativelocation_y _relativelocation_z) (Math|Vector|BreakVector (Variables|/Script/Engine.SceneComponent|GetRelativeLocation_X _staticmeshcomponent)))
  (|CallParentFunction)
  (Networking|CanExecuteCosmeticEvents
    (:True
      (bind _asfort_pickup (Utilities|Casting|CastToFortPickup _returnvalue))
      (Variables|Default|SetWadCount (.Count (Class|FortItemFunctionLibrary|BreakFortItemEntry (Pickup|GetItemEntry _asfort_pickup 0))))
      (Development|PrintString (Utilities|String|Append "Attach Parent Actor" (Utilities|GetDisplayName _returnvalue) "   stack count:" (Utilities|String|ToString(Integer) (.Count (Class|FortItemFunctionLibrary|BreakFortItemEntry (Pickup|GetItemEntry _asfort_pickup 0))))))
      (Utilities|FlowControl|Sequence
        (:then_0
          ;[1 1 1] Large
          (Components|StaticMesh|SetStaticMesh _staticmeshcomponent "/DragonPopGameplayContent/Gameplay/CoinLawless/Meshes/CoinLawless_Pickup.CoinLawless_Pickup")
          (Transformation|SetWorldScale3D _staticmeshcomponent (Math|Conversions|ToVector(Float) (Variables|Default|GetScale)))
          (Transformation|SetRelativeLocation _staticmeshcomponent (Math|Vector|MakeVector _relativelocation_x _relativelocation_y (+ _relativelocation_z 4e+01)) false false)
          (Transformation|SetActorRotation "0.000000,35.000000,0.000000"))))))

(event EventOnPickedUp
  (params (PickupTarget :type Fort%20Pawn%20Object%20Reference))
  (bind _output_get (Variables|Default|SetPickup%20Target PickupTarget))
  (|CallParentFunction _output_get)
  (bind _returnvalue (Niagara|SpawnSystematLocation "/Game/Effects/Fort_Effects/Gameplay/Pickups/Effects/NS_Pickup_Rarity.NS_Pickup_Rarity" (Transformation|GetActorLocation)))
  (Niagara|SetNiagaraVariable(Float) _returnvalue "Scale" 0.3)
  (Niagara|SetNiagaraVariable(LinearColor) _returnvalue "RarityColor" (Math|Conversions|ToLinearColor(Vector) (Math|Conversions|ToVector(LinearColor) "(R=0.000000,G=1.000000,B=0.242000,A=1.000000)"))))

(event EventOnAboutToEnterBackpack
  (params (PickupTarget :type Fort%20Pawn%20Object%20Reference))
  (bind _pickupsound (Variables|Default|GetPickUpSound))
  (Utilities|IsValid PickupTarget
    (:Is%20Valid
      (Utilities|IsValid _pickupsound
        (:Is%20Valid
          (Transformation|AttachComponentToComponent _pickupsound (Variables|/Script/Engine.Actor|GetRootComponent PickupTarget) "root" "SnapToTarget" "SnapToTarget" "SnapToTarget" true)
          (bind _returnvalue (InventoryManagement|GetItemCountforItemDefinition (Fortnite|GetFortPlayerControllerfromActor PickupTarget) "/DragonPopGameplayContent/Gameplay/DiamondCurrency/Athena_DiamondItemData.Athena_DiamondItemData"))
          (Development|PrintString (Utilities|String|ToString(Integer) _returnvalue) true true "(R=1.000000,G=0.000000,B=0.019995,A=1.000000)"))))))

(fn GetWadSizeValue
  (if (>= (Variables|Default|GetWadCount) (Math|Float|Truncate (ScalableFloat|GetValueatLevel (Variables|Default|GetLargeThreshold))))
    ;[1 1 1] Large
    (Variables|Default|SetWadSizeValue 2)
    (elif (>= (Variables|Default|GetWadCount) (Math|Float|Truncate (ScalableFloat|GetValueatLevel (Variables|Default|GetMediumThreshold))))
      ;[1 1 1] Medium
      (Variables|Default|SetWadSizeValue 1)
      (else
        ;[1 1 1] Small
        (Variables|Default|SetWadSizeValue 0))))
  (return (Variables|Default|GetWadSizeValue)))

(fn MarkClosestBlackMarketOnPickUp
  (params (BlackMarketIcons :type Array%20of%20Actor%20Object%20References))
  (bind _pickup_target (Variables|Default|GetPickup%20Target))
  (Networking|SwitchHasAuthority
    (:Authority
      (Utilities|IsValid _pickup_target
        (:Is%20Valid
          (bind _asbp_dragon_pop_icon_manager (Utilities|Casting|CastToBP_DragonPopIconManager (Fortnite|GetClosestActorfromArray _pickup_target BlackMarketIcons)))
          (MarkerComponent|MarkActoronServer (Actor|GetComponentbyClass (Fortnite|GetFortPlayerControllerfromActor _pickup_target) "/Script/FortniteGame.AthenaMarkerComponent") _asbp_dragon_pop_icon_manager false true))))))

(fn SetupPickupGlow
  (params (Level :type Integer))
  (bind _sm_rare_pickup (Variables|Default|Getsm_rare_pickup))
  (Rendering|Material|SetVectorParameterValueonMaterials _sm_rare_pickup "Color" (* "0.500000,1.000000,2.000000" (Utilities|Select :Index Level :Option%200 1.0 :Option%201 1.0 :Option%202 2.0)))
  (Rendering|Material|SetScalarParameterValueonMaterials _sm_rare_pickup "Expansion Lerp" (Utilities|Select :Index Level :Option%200 0.2 :Option%201 0.4 :Option%202 0.0))
  (Rendering|Material|SetScalarParameterValueonMaterials _sm_rare_pickup "Sprite Size" (Utilities|Select :Index Level :Option%200 15.0 :Option%201 25.0 :Option%202 5e+01)))

(fn ConstructionScript
  (|CallParentFunction))
