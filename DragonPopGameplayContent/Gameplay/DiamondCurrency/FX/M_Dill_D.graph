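# Translucent unlit sprite material for diamond currency particle effects, supporting Niagara and Cascade sprites and built around the DillCoin_D texture.
# Particle color RGB runs through a day/night auto bloom control function, then into the shared FX emissive adjustments function.
# A separate branch multiplies the coin texture's RGB by particle alpha, so the coin shape and particle fade combine in one mask.
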Material /DragonPopGameplayContent/Gameplay/DiamondCurrency/FX/M_Dill_D
graph {
 properties {
  bUsedWithNiagaraSprites=true
  bUsedWithParticleSprites=true
  blendMode=BLEND_Translucent
  pixelDepthOffsetMode=PDOM_Legacy
  shadingModel=MSM_Unlit
  usedShadingModels=Unlit
 }
 nodes [
  n1 = ParticleColor()
  MaterialFunctionCall(Emissive=MaterialFunctionCall!(DayBloom_Mult=Constant() {r=1}, In=@n1.RGB) {materialFunction=/Game/Packages/Fortnite_UniqueMaterials_1/MH_Materials/DayNight_Auto_Bloom_Control}.Result) {materialFunction=&MF_FX_EmissiveAdjustments}
  Multiply(A=TextureSample() {texture=&/DillCoin_D}.RGB, B=@n1.A)
 ]
}
