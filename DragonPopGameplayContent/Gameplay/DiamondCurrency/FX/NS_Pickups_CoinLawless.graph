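# Niagara effect for the Lawless coin pickup: a green glitchy glow with rising bits, spinning triangles and a light.
# Glow_Light pulses via nested waveforms, while Glitch_sprite spawns short-lived meshes and Triangles001 emits axis-aligned particles, randomly split between two visibility tags.
# Bits lerp upward, grow with distance, cull relative to User.FXDistance and generate death events.
# BitsEnd's burst is disabled, and no event handler is shown.
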NiagaraSystem /DragonPopGameplayContent/Gameplay/DiamondCurrency/FX/NS_Pickups_CoinLawless
dependencies {
 usedDataInterfaces [&NiagaraDataInterfaceVector2DCurve]
 usedDynamicInputs []
 usedModules [&SystemState]
}
emitters {
 Bits {
  emitterUpdateScript [
   &EmitterState(
    Enable Distance Culling=true
    Inactive Response=::NewEnumerator1
    Life Cycle Mode=::NewEnumerator1
    Loop Duration=1
    MaxDistance:NiagaraFloat=&Multiply_Float(A=$User.FXDistance, B=0.8)
    Module.bUseMaxDistance=true
    Module.bUseMinDistance=true
    Scalability Mode=::NewEnumerator1
    UseLoopDelay=false
   )
   &SpawnRate(Use Spawn Probability=false)
  ]
  particleSpawnScript [
   /SetVariables(Particles.VecAdd:Vector3f=&Lerp_Vector(
     A=<0 0 20>
     B=<0 0 30>
     Alpha=&FloatFromCurve(
      FloatCurve=DataInterface() {propertyValues {Curve {keys [<at 0 0>,<at 0.4 0.2 RCIM_Constant>,<at 0.5 0.8>,<at 1 1>]}}}
      CurveIndex=&RandomRangeFloat(Minimum=0, Maximum=1)
     )
    )) {isSetParameters=true}
   &InitializeParticle(
    Color=<rgba 0 1 0.2417 1>
    Lifetime Mode=::NewEnumerator1
    Sprite Rotation Angle=-22
    Sprite Rotation Mode=::NewEnumerator0
    Sprite Size Mode=::NewEnumerator3
    Uniform Sprite Size=&Multiply_Float(A=&RandomRangeFloat(Minimum=12, Maximum=12), B=1)
    UsePositionOffset=false
    Write Material Random=true
   )
   /SetVariables(
    Particles.GlowSize=<20 20>
    Particles.GlowVisibility:NiagaraInt32=&RandomRangeInt(Minimum=0, Maximum=1)
   ) {isSetParameters=true, module=SetVariables_#1}
   &ShapeLocation(
    Hemisphere Distribution=<1 0.2>
    MaskSpawnGroup=false
    Non Uniform Scale=<1.5 1.5 1>
    Offset Mode=::NewEnumerator0
    Override Seed=false
    Sphere Radius=10
    Sphere Surface Distribution=0.9286
    UseOwnerScale=true
   )
   /SetVariables(Particles.VecEnd:Vector3f=&Add_Vector(A=Linked() {linkedVariable {name=Particles.Position, type=NiagaraPosition}}, B=$Particles.VecAdd)) {isSetParameters=true}
   /SetVariables(Particles.LineLength:NiagaraFloat=&VectorLength(Vector=$Particles.VecAdd)) {isSetParameters=true}
   &AddVelocity(Velocity=&MakeCustomVectorFromBool(
     Bool=&SetBoolByFloatComparison(A=&MakeFloatFromVector(Vector=$Particles.VecAdd, Channel=::NewEnumerator2), B=0)
     True Vector=<0 0 -0.1>
     False Vector=<0 0 0.1>
    ))
  ]
  particleUpdateScript [
   &Drag(Drag=5, Use Linear Drag=true, Use Rotational Drag=false)
   &SolveForcesAndVelocity(Clamp Velocity=false, Limit Acceleration=false)
   &ParticleState()
   &ScaleColor(
    Scale Alpha=&FloatFromCurve(FloatCurve=DataInterface() {propertyValues {Curve {keys [{interpMode=RCIM_Cubic, tangentMode=RCTM_User, time=0, value=0}, {interpMode=RCIM_Cubic, tangentMode=RCTM_User, time=0.3, value=1}, <at 1 1>]}}})
    Scale RGB=&VectorFromFloat(Value=&Lerp_Float(
      A=0.5
      B=1
      Alpha=&FloatFromCurve(
       FloatCurve=DataInterface() {propertyValues {Curve {keys [
           {interpMode=RCIM_Cubic, tangentMode=RCTM_Break, time=0.1, value=0}
           {arriveTangent=3.4325, interpMode=RCIM_Cubic, tangentMode=RCTM_Break, time=1, value=1}
          ]}}}
       CurveIndex=&RandomRangeFloat(Minimum=0, Maximum=1)
      )
     ))
    ScaleA=true
    ScaleRGB=true
   )
   &ScaleSpriteSize(Scale Sprite Size Mode=::NewEnumerator0)
   &ScaleSpriteSize(
    Uniform Curve Index=$Emitter.Scalability.DistanceFraction
    Uniform Curve Sprite Scale=DataInterface() {propertyValues {Curve {keys [
        {interpMode=RCIM_Cubic, tangentMode=RCTM_User, time=0.5, value=1}
        {arriveTangent=7.7563, interpMode=RCIM_Cubic, leaveTangent=7.7563, tangentMode=RCTM_User, time=1, value=3}
       ]}}}
   ) {module=ScaleSpriteSize001}
   /SetVariables(Particles.Position=&Lerp_Vector(
     A=Linked() {linkedVariable {name=Particles.Initial.Position, type=NiagaraPosition}}
     B=$Particles.VecEnd
     Alpha=&FloatFromCurve(FloatCurve=DataInterface() {propertyValues {Curve {keys [<at 0 0>,<at 1 1>]}}})
    )) {isSetParameters=true}
   &GenerateDeathEvent(Event Probability=1, Use Event Probability=true)
   /SetVariables(Particles.LineSize:Vector2f=&MakeVector2D(
     X=1
     Y=&Lerp_Float(
      A=1
      B=$Particles.LineLength
      Alpha=&FloatFromCurve(FloatCurve=DataInterface() {propertyValues {Curve {keys [<at 0 0>,<at 1 1>]}}})
     )
    )) {isSetParameters=true}
   /SetVariables(Particles.LineColor:LinearColor=&Multiply_LinearColorByFloat(Linear Color=$Particles.Initial.Color, Float=0.05)) {isSetParameters=true}
   /SetVariables() {isSetParameters=true}
  ]
  rendererClasses [&NiagaraSpriteRendererProperties]
  renderers [{rendererClass=&NiagaraSpriteRendererProperties}]
 }
 BitsEnd {
  emitterUpdateScript [
   &EmitterState(Inactive Response=::NewEnumerator1, Life Cycle Mode=::NewEnumerator1, Loop Duration=2.5, UseLoopDelay=false)
   &SpawnBurst_Instantaneous(Use Spawn Probability=false, UseLoopCountLimit=false) {enabled=false}
  ]
  particleSpawnScript [
   &InitializeParticle(
    Color=<rgba 0 1 0.2334 1>
    Lifetime Max=0.8
    Lifetime Min=0.5
    Lifetime Mode=::NewEnumerator1
    Sprite Rotation Angle=-22
    Sprite Rotation Mode=::NewEnumerator0
    Sprite Size Mode=::NewEnumerator3
    Uniform Sprite Size=&Multiply_Float(A=&RandomRangeFloat(Minimum=12, Maximum=12), B=1)
    UsePositionOffset=false
    Write Material Random=true
   )
   /SetVariables(
    Particles.GlowSize=<20 20>
    Particles.GlowVisibility:NiagaraInt32=&RandomRangeInt(Minimum=0, Maximum=1)
   ) {isSetParameters=true, module=SetVariables_#1}
   &AddVelocity(Velocity=&Multiply_Vector(
     A=&Lerp_Vector(
      A=<-10 -10 0>
      B=<10 10 0>
      Alpha=&FloatFromCurve(
       FloatCurve=DataInterface() {propertyValues {Curve {keys [<at 0 0>,<at 0.4 0.2 RCIM_Constant>,<at 0.5 0.8>,<at 1 1>]}}}
       CurveIndex=&RandomRangeFloat(Minimum=0, Maximum=1)
      )
     )
     B=&Lerp_Vector(
      A=<1 0 0>
      B=<0 1 0>
      Alpha=&FloatFromCurve(
       FloatCurve=DataInterface() {propertyValues {Curve {keys [<at 0 0 RCIM_Constant>,<at 0.5 1 RCIM_Constant>]}}}
       CurveIndex=&RandomRangeFloat(Minimum=0, Maximum=1)
      )
     )
    ))
  ]
  particleUpdateScript [
   &SolveForcesAndVelocity(Clamp Velocity=false, Limit Acceleration=false)
   &ParticleState()
   &ScaleColor(
    Scale RGB=&VectorFromFloat(Value=&Lerp_Float(
      A=15
      B=15
      Alpha=&FloatFromCurve(
       FloatCurve=DataInterface() {propertyValues {Curve {keys [
           {interpMode=RCIM_Cubic, tangentMode=RCTM_Break, time=0.1, value=0}
           {arriveTangent=3.4325, interpMode=RCIM_Cubic, tangentMode=RCTM_Break, time=1, value=1}
          ]}}}
       CurveIndex=&RandomRangeFloat(Minimum=0, Maximum=1)
      )
     ))
    ScaleA=true
    ScaleRGB=true
   ) {enabled=false}
   &ScaleSpriteSize(Uniform Curve Sprite Scale=DataInterface() {propertyValues {Curve {keys [
        {arriveTangent=5.8335, time=0, value=1}
        {interpMode=RCIM_Cubic, tangentMode=RCTM_User, time=0.8, value=1}
        {arriveTangent=-12.5579, interpMode=RCIM_Cubic, tangentMode=RCTM_Break, time=1, value=0.2}
       ]}}})
   &ScaleSpriteSize(
    Uniform Curve Index=$Emitter.Scalability.DistanceFraction
    Uniform Curve Sprite Scale=DataInterface() {propertyValues {Curve {keys [
        {interpMode=RCIM_Cubic, tangentMode=RCTM_User, time=0.5, value=1}
        {arriveTangent=7.7563, interpMode=RCIM_Cubic, leaveTangent=7.7563, tangentMode=RCTM_User, time=1, value=3}
       ]}}}
   ) {module=ScaleSpriteSize001}
  ]
  rendererClasses [&NiagaraSpriteRendererProperties]
  renderers [{rendererClass=&NiagaraSpriteRendererProperties}]
 }
 Glitch_sprite {
  emitterUpdateScript [
   &EmitterState()
   &SpawnRate(SpawnRate=10, Use Spawn Probability=false)
  ]
  particleSpawnScript [
   &InitializeParticle(
    Color=<rgba 0 5 1.3182 2>
    Lifetime Max=0.3
    Lifetime Min=0.3
    Lifetime Mode=::NewEnumerator1
    Mesh Scale Min=<1 1 3>
    Mesh Scale Mode=::NewEnumerator2
    Position Offset=<0 0 100>
    UsePositionOffset=false
    Write Material Random=true
   )
   &CameraOffset(Camera Offset Amount=-30)
   &ShapeLocation(MaskSpawnGroup=false, Override Seed=false, Sphere Radius=10, Sphere Surface Distribution=0.7024, UseOwnerScale=true)
   &AddVelocity(Velocity=&Lerp_Vector(
     A=<0 0 -20>
     B=<0 0 20>
     Alpha=&FloatFromCurve(
      FloatCurve=DataInterface() {propertyValues {Curve {keys [<at 0 0 RCIM_Constant>,<at 0.5 1 RCIM_Constant>]}}}
      CurveIndex=&RandomRangeFloat(Minimum=0, Maximum=1)
     )
    ))
  ]
  particleUpdateScript [
   &ParticleState()
   &ScaleMeshSize(Scale Factor=&Multiply_VectorByFloat(
     Vector=<1 1 1>
     Float=&Multiply_Float(
      A=0.4
      B=&FloatFromCurve(
       FloatCurve=DataInterface() {propertyValues {Curve {keys [
           {arriveTangent=4.1785, time=0, value=1}
           <at 0.5 1.1>
           {arriveTangent=-0.0405, interpMode=RCIM_Cubic, leaveTangent=-0.0405, tangentMode=RCTM_User, time=1, value=1.5}
          ]}}}
       Scale Curve=0.75
      )
     )
    ))
   &Drag(Drag=2, Use Linear Drag=true, Use Rotational Drag=false)
   &SolveForcesAndVelocity(Clamp Velocity=false, Limit Acceleration=false)
   &DynamicMaterialParameters(Module.Param0WriteEnabled=true, Module.Param1WriteEnabled=true, Module.Param2WriteEnabled=true, Module.Param3WriteEnabled=true)
   &ScaleColor(
    Scale Alpha=&Multiply_Float(
     A=&FloatFromCurve(
      FloatCurve=DataInterface() {propertyValues {Curve {keys [
          {arriveTangent=9.0725, interpMode=RCIM_Cubic, leaveTangent=9.0725, tangentMode=RCTM_User, time=0, value=0}
          {arriveTangent=-2.5249, interpMode=RCIM_Cubic, leaveTangent=-2.5249, tangentMode=RCTM_User, time=0.2, value=0.9847}
          {arriveTangent=-0.0405, interpMode=RCIM_Cubic, leaveTangent=-0.0405, tangentMode=RCTM_User, time=1, value=0}
         ]}}}
      Scale Curve=&RandomRangeFloat(Minimum=0.2, Maximum=1)
     )
     B=0.5
    )
    ScaleA=true
    ScaleRGB=true
   )
  ]
  rendererClasses [&NiagaraMeshRendererProperties]
  renderers [{rendererClass=&NiagaraMeshRendererProperties, rendererIndex=0}]
 }
 Glow_Light {
  emitterUpdateScript [
   &EmitterState(Life Cycle Mode=::NewEnumerator1, Loop Behavior=::NewEnumerator1, Loop Duration Mode=::NewEnumerator1, UseLoopDelay=false)
   &SpawnBurst_Instantaneous(Use Spawn Probability=false, UseLoopCountLimit=false)
  ]
  particleSpawnScript [
   &InitializeParticle(
    Color=<rgba 0.7085 1.3 0.117 1>
    Sprite Size Mode=::NewEnumerator3
    Uniform Sprite Size=210
    UsePositionOffset=false
    Write Material Random=true
   )
   &CameraOffset(Camera Offset Amount=-30)
   /SetVariables(
    Particles.DarkGlowCamOffset=-30
    Particles.DarkGlowColor:LinearColor=<rgba 0 1 0.1652 0.3>
    Particles.DarkGlowDynPar:Vector4f=<3 0 0 0>
    Particles.DarkGlowPos:NiagaraPosition=&AddVectorToPosition(Position=/Niagara/DynamicInputs/Helpers/SimulationPosition(), Vector=<0 0 0>)
    Particles.DarkGlowSize:Vector2f=&Vector2DFromFloat(Value=100)
   ) {isSetParameters=true}
  ]
  particleUpdateScript [
   &ParticleState(Kill Particles When Lifetime Has Elapsed=false, Let Infinitely Lived Particles Die When Emitter Deactivates=true)
   &ScaleColor(
    Scale Alpha=&Waveform(
     "Waveform [1]"=::NewEnumerator0
     "[1] Frequency"=0.7
     "[1] Amplitude Min/Max"=<0.3 0.7>
     "[1] Scale"=&Waveform("Waveform [1]"=::NewEnumerator0, "[1] Frequency"=1.2, "[1] Amplitude Min/Max"=<0.3 0.7>)
    )
    ScaleA=true
    ScaleRGB=true
   )
   &DynamicMaterialParameters(
    Index 0 Param 1=1
    Index 0 Param 2=10
    Module.Index0Param0Write=true
    Module.Index0Param1Write=true
    Module.Index0Param2Write=false
    Module.Index0Param3Write=false
   )
   &FN_Light_Attributes(Color=$Particles.Color, Create Color Attribute=true, Create Exponent Attribute=true, Scale Color=4)
  ]
  rendererClasses [&NiagaraSpriteRendererProperties, &NiagaraLightRendererProperties]
  renderers [
   {rendererClass=&NiagaraSpriteRendererProperties}
   {rendererClass=&NiagaraLightRendererProperties, rendererIndex=1}
   {rendererClass=&NiagaraSpriteRendererProperties, rendererIndex=2}
  ]
 }
 Triangles001 {
  emitterUpdateScript [
   &EmitterState()
   &SpawnRate(SpawnRate=15, Use Spawn Probability=false)
   /SetVariables(Emitter.VecRand:Vector3f=&RandomRangeVector(Minimum=<-8 -8 -8>, Maximum=<8 8 8>)) {isSetParameters=true}
  ]
  particleSpawnScript [
   /SetVariables(Particles.Random=&RandomRangeFloat(Minimum=0, Maximum=1)) {isSetParameters=true}
   /Niagara/Modules/Spawn/Initialization/InitializeParticle(
    Color=<rgba 0 5 1.3182 2>
    Lifetime=&Lerp_Float(A=1.5, B=2, Alpha=$Particles.Random)
    Module.Write Color=true
    Module.Write Lifetime=true
    Module.Write Mass=true
    Module.Write Material Random=true
    Module.Write Position=false
    Module.Write Scale=false
    Module.Write SpriteRotation=true
    Module.Write SpriteSize=true
    Position=/Niagara/Functions/Localspace/SimulationPosition()
    Sprite Rotation=&Lerp_Float(A=0, B=360, Alpha=$Particles.Random)
    Sprite Size=&Lerp_Vector2(A=<2.5 2.5>, B=<1 1>, Alpha=$Particles.Random)
   )
   &ShapeLocation(MaskSpawnGroup=false, Override Seed=false, Sphere Radius=10, Sphere Surface Distribution=1, UseOwnerScale=true) {module=ShapeLocation001}
   &AddVelocity(Velocity=&Multiply_Vector(
     A=&Lerp_Vector(
      A=<-15 -15 10>
      B=<15 15 20>
      Alpha=&FloatFromCurve(
       FloatCurve=DataInterface() {propertyValues {Curve {keys [<at 0 0>,<at 0.4 0.2 RCIM_Constant>,<at 0.5 0.8>,<at 1 1>]}}}
       CurveIndex=&RandomRangeFloat(Minimum=0, Maximum=1)
      )
     )
     B=&MakeCustomVectorFromBool(
      Bool=&RandomBool()
      True Vector=&MakeCustomVectorFromBool(Bool=&RandomBool(), True Vector=<0 0 1>, False Vector=<0 1 0>)
      False Vector=<1 0 0>
     )
    ))
   /SetVariables(
    Particles.GlowSize=&RandomRangeVector2D(Minimum=<1.5 1.5>, Maximum=<1 1>)
    Particles.GlowVisibility:NiagaraInt32=&RandomRangeInt(Minimum=0, Maximum=1)
   ) {isSetParameters=true}
   /SetVariables(Particles.TriVisibility:NiagaraInt32=&MakeCustomIntFromBool(
     Bool=&SetBoolByIntComparison(A=$Particles.GlowVisibility, B=0, Comparison Type=::NewEnumerator0)
     True INT=0
     False INT=1
    )) {isSetParameters=true}
  ]
  particleUpdateScript [
   &ParticleState()
   &ScaleSpriteSize(Scale Factor=&Vector2DFromCurve(Vector2Curve=DataInterface() {propertyValues {XCurve {keys [<at 0 0>,<at 0.4 1>]}, YCurve {keys [<at 0 1>,<at 1 1>]}}}, ScaleCurve=<1 1>))
   &DynamicMaterialParameters(
    Index 0 Param 1=&FloatFromCurve(FloatCurve=DataInterface() {propertyValues {Curve {keys [<at 0.5 0.1>,<at 1 0>]}}})
    Module.Param0WriteEnabled=true
    Module.Param1WriteEnabled=true
    Module.Param2WriteEnabled=true
    Module.Param3WriteEnabled=true
   )
   &SpriteRotationRate(Rotation Rate=&Lerp_Float(
     A=-75
     B=75
     Alpha=&Multiply_Float(
      A=$Particles.Random
      B=&FloatFromCurve(FloatCurve=DataInterface() {propertyValues {Curve {keys [<at 0.4 1>,<at 1 0.1>]}}})
     )
    ))
   &ScaleColor(
    Scale Alpha=&FloatFromCurve(FloatCurve=DataInterface() {propertyValues {Curve {keys [
         {interpMode=RCIM_Cubic, leaveTangentWeight=0.3895, tangentMode=RCTM_User, time=0.5, value=1}
         {arriveTangent=-5.3768, tangentMode=RCTM_User, time=1, value=0}
        ]}}})
    Scale RGB=&VectorFromFloat(Value=&FloatFromCurve(FloatCurve=DataInterface() {propertyValues {Curve {keys [<at 0 1>,<at 0.15 50>,<at 0.2937 0.6936>,<at 1 0>]}}}))
    ScaleA=true
    ScaleRGB=true
   )
   &SolveForcesAndVelocity(Clamp Velocity=false, Limit Acceleration=false)
   /SetVariables(Particles.GlowColor=&Multiply_LinearColorByFloat(Linear Color=$Particles.Color, Float=0.35)) {isSetParameters=true}
  ]
  rendererClasses [&NiagaraSpriteRendererProperties]
  renderers [
   {rendererClass=&NiagaraSpriteRendererProperties}
   {rendererClass=&NiagaraSpriteRendererProperties, rendererIndex=1}
  ]
 }
}
summary {
 systemName=NS_Pickups_CoinLawless
 emitters [
  {emitterName=Glitch_sprite, rendererClasses [&NiagaraMeshRendererProperties]}
  {emitterName=Triangles001, rendererClasses [&NiagaraSpriteRendererProperties]}
  {emitterName=Bits, rendererClasses [&NiagaraSpriteRendererProperties]}
  {emitterName=BitsEnd, rendererClasses [&NiagaraSpriteRendererProperties]}
  {emitterName=Glow_Light, rendererClasses [&NiagaraSpriteRendererProperties, &NiagaraLightRendererProperties]}
 ]
 userVariables [
  User.BaseCurveColor {
   type=NiagaraDataInterfaceColorCurve
   description=""
   defaultValue {
    struct=NiagaraExt_VariableValue_DataInterface
    value {
     dataInterface=&self:NiagaraDataInterfaceColorCurve_1
     dataInterfaceClass=&NiagaraDataInterfaceColorCurve
    }
   }
  }
  User.BaseCurveColorHueShift {
   type=NiagaraDataInterfaceCurve
   description=""
   defaultValue {
    struct=NiagaraExt_VariableValue_DataInterface
    value {
     dataInterface=&self:NiagaraDataInterfaceCurve_1
     dataInterfaceClass=&NiagaraDataInterfaceCurve
    }
   }
  }
  User.BaseCurveColorPick {type=NiagaraFloat, defaultValue=0, description=""}
  User.bIsFrontEnd {type=NiagaraBool}
  User.bIsFrontEndPreview {type=NiagaraBool}
  User.FXColorRed {type=NiagaraBool}
  User.FXDistance {type=NiagaraFloat, defaultValue=8000, description=""}
  User.FXScale {type=NiagaraFloat, defaultValue=1, description=""}
  User.FXScaleXYZ {type=Vector3f, description="", defaultValue {struct=Vector3f, value=<1 1 1>}}
 ]
}
